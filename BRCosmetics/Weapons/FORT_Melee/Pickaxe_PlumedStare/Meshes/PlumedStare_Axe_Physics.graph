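# Physics asset for the PlumedStare two-handed pickaxe, setting up physical motion for three rings and a forked tassel.
# Capsule bodies cover the handle and each chain bone, with tiny base capsules anchoring the chains to the handle.
# Rings swing only a few degrees, while the tassel splits into two-segment front and back strands with wide swing and tight twist limits.
# The tassel's base joint is angularly free against the handle.
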PhysicsAsset /BRCosmetics/Weapons/FORT_Melee/Pickaxe_PlumedStare/Meshes/PlumedStare_Axe_Physics
bodies {
 Melee_TwoHanded_Handle [{
   shapeType=Capsule
   length=54.4646
   radius=8.9696
   shapeName=Melee_TwoHanded_Handle_capsule
   center=<-0.0069 -1.2016 33.2752>
   rotation=<rot -2.753 -90 -0.0171>
  }]
 base_ring1_c [{shapeType=Capsule, length=0.505, radius=0.505, shapeName=base_ring1_c_capsule, rotation=<rot 0 -90 0>}]
 base_ring2_c [{shapeType=Capsule, length=0.505, radius=0.505, shapeName=base_ring2_c_capsule, rotation=<rot 0 -90 0>}]
 base_ring3_c [{shapeType=Capsule, length=0.505, radius=0.505, shapeName=base_ring3_c_capsule, rotation=<rot 0 -90 0>}]
 base_tassel_c [{shapeType=Capsule, length=0.505, radius=0.505, shapeName=base_tassel_c_capsule, rotation=<rot 0 -90 0>}]
 dyn_ring1_c [{
   shapeType=Capsule
   length=1.9652
   radius=0.6296
   shapeName=dyn_ring1_c_capsule
   center=<2.0259 0 -2.2948>
   rotation=<rot 57.0474 0.0013 0.0008>
  }]
 dyn_ring2_c [{
   shapeType=Capsule
   length=2.1426
   radius=0.5883
   shapeName=dyn_ring2_c_capsule
   center=<2.8658 0 -0.0992>
   rotation=<rot 87.142 179.991 179.9907>
  }]
 dyn_ring3_c [{
   shapeType=Capsule
   length=2.2433
   radius=0.5898
   shapeName=dyn_ring3_c_capsule
   center=<2.0707 0 1.7583>
   rotation=<rot 58.5083 179.9995 179.9992>
  }]
 dyn_tassel_00_bk [{shapeType=Capsule, length=3.5302, radius=1, shapeName=dyn_tassel_00_bk_capsule, center=<0.3815 0 0>, rotation=<rot 90 0 0>}]
 dyn_tassel_00_fr [{shapeType=Capsule, length=2.7959, radius=1, shapeName=dyn_tassel_00_fr_capsule, center=<1.2068 0 0.1585>, rotation=<rot 90 0 0>}]
 dyn_tassel_01_bk [{shapeType=Capsule, length=3.2667, radius=1.3414, shapeName=dyn_tassel_01_bk_capsule, center=<3.5991 0 0>, rotation=<rot 90 0 0>}]
 dyn_tassel_01_fr [{shapeType=Capsule, length=3.3333, radius=1.3414, shapeName=dyn_tassel_01_fr_capsule, center=<3.5991 0 0>, rotation=<rot 90 0 0>}]
 dyn_tassel_c [{
   shapeType=Capsule
   length=3.8141
   radius=1.2014
   shapeName=dyn_tassel_c_capsule
   center=<1.5827 0 -0.0062>
   rotation=<rot 80.7449 0.0003 0.0001>
  }]
}
constraints [
 {bone1Name=base_ring1_c, bone2Name=Melee_TwoHanded_Handle, swing1LimitDegrees=7, swing2LimitDegrees=2, twistLimitDegrees=1}
 {bone1Name=base_ring2_c, bone2Name=Melee_TwoHanded_Handle, swing1LimitDegrees=7, swing2LimitDegrees=2, twistLimitDegrees=1}
 {bone1Name=base_ring3_c, bone2Name=Melee_TwoHanded_Handle, swing1LimitDegrees=7, swing2LimitDegrees=2, twistLimitDegrees=1}
 {bone1Name=dyn_tassel_01_bk, bone2Name=dyn_tassel_00_bk, swing1LimitDegrees=55, swing2LimitDegrees=55, twistLimitDegrees=1}
 {bone1Name=dyn_tassel_01_fr, bone2Name=dyn_tassel_00_fr, swing1LimitDegrees=55, swing2LimitDegrees=55, twistLimitDegrees=1}
 {bone1Name=dyn_tassel_00_bk, bone2Name=dyn_tassel_c, swing1LimitDegrees=85, swing2LimitDegrees=85, twistLimitDegrees=1}
 {bone1Name=dyn_tassel_00_fr, bone2Name=dyn_tassel_c, swing1LimitDegrees=85, swing2LimitDegrees=85, twistLimitDegrees=1}
 {bone1Name=dyn_tassel_c, bone2Name=base_tassel_c, swing1LimitDegrees=55, swing2LimitDegrees=55, twistLimitDegrees=1}
 {bone1Name=dyn_ring1_c, bone2Name=base_ring1_c, swing1LimitDegrees=5, swing2LimitDegrees=8, twistLimitDegrees=1}
 {bone1Name=dyn_ring2_c, bone2Name=base_ring2_c, swing1LimitDegrees=5, swing2LimitDegrees=8, twistLimitDegrees=1}
 {bone1Name=dyn_ring3_c, bone2Name=base_ring3_c, swing1LimitDegrees=5, swing2LimitDegrees=8, twistLimitDegrees=1}
 {
  bone1Name=base_tassel_c
  bone2Name=Melee_TwoHanded_Handle
  swing1LimitDegrees=45
  swing1Motion=Free
  swing2LimitDegrees=45
  swing2Motion=Free
  twistLimitDegrees=45
  twistMotion=Free
 }
]
